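# Nitro Queen arena PA announcer cue for NitroDrome, named for AI spawns though its voice lines are player-enters-vehicle takes.
# A Branch on the BossSpawned parameter plays only crowd cheers when true.
# When false or unset, it mixes a random announcer line with a quieter random cheer.
# Cheers follow a fixed delay on both paths, and shared announcer attenuation and concurrency settings apply.
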SoundCue /NitroDrome/Assets/Audio/PA/VO/Nitro_Queen/VO_Nitro_Queen_PA_AI_Spawn_Cue
root=Branch_0
nodes [
 {id=Branch_0, children [Mixer_1,Mixer_0,Mixer_0], properties {boolParameterName=BossSpawned}}
 {id=Delay_0, children [Random_1], properties {delayMax=2, delayMin=2}}
 {id=Mixer_0, children [Random_0,Delay_0], properties {inputVolume [0.85,0.6]}}
 {id=Mixer_1, children [Delay_0], properties {inputVolume [0.6]}}
 {
  id=Random_0
  children [WavePlayer_0,WavePlayer_1,WavePlayer_2,WavePlayer_3,WavePlayer_4]
  properties {bShouldExcludeFromBranchCulling=true, weights [1,1,1,1,1]}
 }
 {
  id=Random_1
  children [WavePlayer_5, WavePlayer_6, WavePlayer_7, WavePlayer_8, WavePlayer_9, WavePlayer_10, WavePlayer_11, WavePlayer_12]
  properties {bShouldExcludeFromBranchCulling=true, weights [1,1,1,1,1,1,1,1]}
 }
 {id=WavePlayer_0, properties {soundWaveAssetPtr=&/VO_Nitro_Queen_PA_Player_Enters_Veh_01}}
 {id=WavePlayer_1, properties {soundWaveAssetPtr=&/VO_Nitro_Queen_PA_Player_Enters_Veh_02}}
 {id=WavePlayer_10, properties {soundWaveAssetPtr=/NitroDrome/Assets/Audio/PA/Crowd/NitroDrome_PA_Crowd_Cheer_06}}
 {id=WavePlayer_11, properties {soundWaveAssetPtr=/NitroDrome/Assets/Audio/PA/Crowd/NitroDrome_PA_Crowd_Cheer_07}}
 {id=WavePlayer_12, properties {soundWaveAssetPtr=/NitroDrome/Assets/Audio/PA/Crowd/NitroDrome_PA_Crowd_Cheer_08}}
 {id=WavePlayer_2, properties {soundWaveAssetPtr=&/VO_Nitro_Queen_PA_Player_Enters_Veh_03}}
 {id=WavePlayer_3, properties {soundWaveAssetPtr=&/VO_Nitro_Queen_PA_Player_Enters_Veh_04}}
 {id=WavePlayer_4, properties {soundWaveAssetPtr=&/VO_Nitro_Queen_PA_Player_Enters_Veh_05}}
 {id=WavePlayer_5, properties {soundWaveAssetPtr=/NitroDrome/Assets/Audio/PA/Crowd/NitroDrome_PA_Crowd_Cheer_01}}
 {id=WavePlayer_6, properties {soundWaveAssetPtr=/NitroDrome/Assets/Audio/PA/Crowd/NitroDrome_PA_Crowd_Cheer_02}}
 {id=WavePlayer_7, properties {soundWaveAssetPtr=/NitroDrome/Assets/Audio/PA/Crowd/NitroDrome_PA_Crowd_Cheer_03}}
 {id=WavePlayer_8, properties {soundWaveAssetPtr=/NitroDrome/Assets/Audio/PA/Crowd/NitroDrome_PA_Crowd_Cheer_04}}
 {id=WavePlayer_9, properties {soundWaveAssetPtr=/NitroDrome/Assets/Audio/PA/Crowd/NitroDrome_PA_Crowd_Cheer_05}}
]
properties {
 attenuationSettings=/NitroDrome/Assets/Audio/PA/VO/VO_PA_Announcer_Att
 maxDistance=20000
 priority=1.05
 soundClassObject=&Fort_VO_NPC
 volumeMultiplier=0.85
 concurrencySet [/NitroDrome/Assets/Audio/PA/VO/VO_PA_Announcer_Conc]
}
