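# Sound cue for the NitroDrome PA boss-eliminated event, playing a buzzer and crowd cheers; despite its Nitro Queen VO name, it references no voice line.
# A mixer layers the full-volume buzzer with two quieter, fixed-delay random branches, each choosing from four different cheers, staggered a second apart.
# It uses the shared PA announcer attenuation and concurrency and the NPC voice sound class.
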SoundCue /NitroDrome/Assets/Audio/PA/VO/Nitro_Queen/VO_Nitro_Queen_PA_Boss_Eliminated_Cue
root=Mixer_0
nodes [
 {id=Delay_1, children [Random_3], properties {delayMax=5, delayMin=5}}
 {id=Delay_2, children [Random_4], properties {delayMax=6, delayMin=6}}
 {id=Mixer_0, children [WavePlayer_0,Delay_1,Delay_2], properties {inputVolume [1,0.3,0.2]}}
 {
  id=Random_3
  children [WavePlayer_17, WavePlayer_18, WavePlayer_19, WavePlayer_20]
  properties {bShouldExcludeFromBranchCulling=true, weights [1,1,1,1]}
 }
 {
  id=Random_4
  children [WavePlayer_25, WavePlayer_26, WavePlayer_27, WavePlayer_28]
  properties {bShouldExcludeFromBranchCulling=true, weights [1,1,1,1]}
 }
 {id=WavePlayer_0, properties {soundWaveAssetPtr=/NitroDrome/Assets/Audio/NitroDrome_Buzzer_Boss_Elimnated_01}}
 {id=WavePlayer_17, properties {soundWaveAssetPtr=/NitroDrome/Assets/Audio/PA/Crowd/NitroDrome_PA_Crowd_Cheer_01}}
 {id=WavePlayer_18, properties {soundWaveAssetPtr=/NitroDrome/Assets/Audio/PA/Crowd/NitroDrome_PA_Crowd_Cheer_02}}
 {id=WavePlayer_19, properties {soundWaveAssetPtr=/NitroDrome/Assets/Audio/PA/Crowd/NitroDrome_PA_Crowd_Cheer_03}}
 {id=WavePlayer_20, properties {soundWaveAssetPtr=/NitroDrome/Assets/Audio/PA/Crowd/NitroDrome_PA_Crowd_Cheer_04}}
 {id=WavePlayer_25, properties {soundWaveAssetPtr=/NitroDrome/Assets/Audio/PA/Crowd/NitroDrome_PA_Crowd_Cheer_05}}
 {id=WavePlayer_26, properties {soundWaveAssetPtr=/NitroDrome/Assets/Audio/PA/Crowd/NitroDrome_PA_Crowd_Cheer_06}}
 {id=WavePlayer_27, properties {soundWaveAssetPtr=/NitroDrome/Assets/Audio/PA/Crowd/NitroDrome_PA_Crowd_Cheer_07}}
 {id=WavePlayer_28, properties {soundWaveAssetPtr=/NitroDrome/Assets/Audio/PA/Crowd/NitroDrome_PA_Crowd_Cheer_08}}
]
properties {
 attenuationSettings=/NitroDrome/Assets/Audio/PA/VO/VO_PA_Announcer_Att
 maxDistance=20000
 priority=1.05
 soundClassObject=&Fort_VO_NPC
 volumeMultiplier=1
 concurrencySet [/NitroDrome/Assets/Audio/PA/VO/VO_PA_Announcer_Conc]
}
