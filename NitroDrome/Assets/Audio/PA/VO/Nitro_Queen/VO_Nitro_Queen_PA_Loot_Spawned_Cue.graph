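# Sound cue for the NitroDrome PA loot-spawn announcement, featuring Nitro Queen voice lines, a checkpoint horn and crowd cheers.
# A BossSpawn bool branch drops the voice line when true, leaving horn and cheer; false or unset mixes all three.
# The horn plays immediately, with fixed delays before the randomly chosen cheer and, slightly later, the randomly chosen voice line.
# Playback uses the shared PA announcer attenuation and concurrency.
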SoundCue /NitroDrome/Assets/Audio/PA/VO/Nitro_Queen/VO_Nitro_Queen_PA_Loot_Spawned_Cue
root=Branch_0
nodes [
 {id=Branch_0, children [Mixer_2,Mixer_0,Mixer_0], properties {boolParameterName=BossSpawn}}
 {id=Delay_0, children [Random_1], properties {delayMax=2.35, delayMin=2.35}}
 {id=Delay_2, children [Random_0], properties {delayMax=2.75, delayMin=2.75}}
 {id=Mixer_0, children [Delay_2,Delay_0,WavePlayer_0], properties {inputVolume [0.85,0.6,1.2]}}
 {id=Mixer_2, children [Delay_0,WavePlayer_0], properties {inputVolume [0.6,1.2]}}
 {id=Random_0, children [WavePlayer_1,WavePlayer_2], properties {bShouldExcludeFromBranchCulling=true, weights [1,1]}}
 {
  id=Random_1
  children [WavePlayer_4, WavePlayer_5, WavePlayer_6, WavePlayer_7, WavePlayer_8, WavePlayer_10, WavePlayer_11, WavePlayer_12]
  properties {bShouldExcludeFromBranchCulling=true, weights [1,1,1,1,1,1,1,1]}
 }
 {id=WavePlayer_0, properties {soundWaveAssetPtr=/NitroDrome/Assets/Audio/Checkpoints/NitroDrome_Checkpoint_Completed_Horn_02}}
 {id=WavePlayer_1, properties {soundWaveAssetPtr=&/VO_Nitro_Queen_PA_Loot_Spawn_01}}
 {id=WavePlayer_10, properties {soundWaveAssetPtr=/NitroDrome/Assets/Audio/PA/Crowd/NitroDrome_PA_Crowd_Cheer_06}}
 {id=WavePlayer_11, properties {soundWaveAssetPtr=/NitroDrome/Assets/Audio/PA/Crowd/NitroDrome_PA_Crowd_Cheer_07}}
 {id=WavePlayer_12, properties {soundWaveAssetPtr=/NitroDrome/Assets/Audio/PA/Crowd/NitroDrome_PA_Crowd_Cheer_08}}
 {id=WavePlayer_2, properties {soundWaveAssetPtr=&/VO_Nitro_Queen_PA_Loot_Spawn_02}}
 {id=WavePlayer_4, properties {soundWaveAssetPtr=/NitroDrome/Assets/Audio/PA/Crowd/NitroDrome_PA_Crowd_Cheer_01}}
 {id=WavePlayer_5, properties {soundWaveAssetPtr=/NitroDrome/Assets/Audio/PA/Crowd/NitroDrome_PA_Crowd_Cheer_02}}
 {id=WavePlayer_6, properties {soundWaveAssetPtr=/NitroDrome/Assets/Audio/PA/Crowd/NitroDrome_PA_Crowd_Cheer_03}}
 {id=WavePlayer_7, properties {soundWaveAssetPtr=/NitroDrome/Assets/Audio/PA/Crowd/NitroDrome_PA_Crowd_Cheer_04}}
 {id=WavePlayer_8, properties {soundWaveAssetPtr=/NitroDrome/Assets/Audio/PA/Crowd/NitroDrome_PA_Crowd_Cheer_05}}
]
properties {
 attenuationSettings=/NitroDrome/Assets/Audio/PA/VO/VO_PA_Announcer_Att
 maxDistance=20000
 priority=1.25
 soundClassObject=&Fort_VO_NPC
 volumeMultiplier=1
 concurrencySet [/NitroDrome/Assets/Audio/PA/VO/VO_PA_Announcer_Conc]
}
