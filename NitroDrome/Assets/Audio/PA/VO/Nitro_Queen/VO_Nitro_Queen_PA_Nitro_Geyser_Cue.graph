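# Sound cue for the NitroDrome PA reacting to a Nitro Geyser, pairing Nitro Queen fire-hazard announcer lines with crowd reactions.
# After a short lead-in delay, a BossSpawn branch plays only quieter crowd audio when true, or a random VO line mixed over the crowd when false or unset.
# The shared crowd layer mixes a random cheer with a later, randomly delayed shock, each on its own delay.
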SoundCue /NitroDrome/Assets/Audio/PA/VO/Nitro_Queen/VO_Nitro_Queen_PA_Nitro_Geyser_Cue
root=Delay_2
nodes [
 {id=Branch_0, children [Mixer_2,Mixer_0,Mixer_0], properties {boolParameterName=BossSpawn}}
 {id=Delay_0, children [Random_2], properties {delayMax=0.25, delayMin=0.25}}
 {id=Delay_1, children [Random_3], properties {delayMax=2, delayMin=1}}
 {id=Delay_2, children [Branch_0], properties {delayMax=0.75, delayMin=0.75}}
 {id=Mixer_0, children [Random_1,Mixer_1], properties {inputVolume [0.85,0.6]}}
 {id=Mixer_1, children [Delay_0,Delay_1], properties {inputVolume [1,1]}}
 {id=Mixer_2, children [Mixer_1], properties {inputVolume [0.6]}}
 {
  id=Random_1
  children [WavePlayer_4,WavePlayer_5,WavePlayer_6,WavePlayer_7,WavePlayer_8]
  properties {bShouldExcludeFromBranchCulling=true, weights [1,1,1,1,1]}
 }
 {
  id=Random_2
  children [WavePlayer_14, WavePlayer_15, WavePlayer_16, WavePlayer_17, WavePlayer_18, WavePlayer_10, WavePlayer_11, WavePlayer_12]
  properties {bShouldExcludeFromBranchCulling=true, weights [1,1,1,1,1,1,1,1]}
 }
 {
  id=Random_3
  children [WavePlayer_22, WavePlayer_23, WavePlayer_24, WavePlayer_25, WavePlayer_26, WavePlayer_27]
  properties {bShouldExcludeFromBranchCulling=true, weights [1,1,1,1,1,1]}
 }
 {id=WavePlayer_10, properties {soundWaveAssetPtr=/NitroDrome/Assets/Audio/PA/Crowd/NitroDrome_PA_Crowd_Cheer_06}}
 {id=WavePlayer_11, properties {soundWaveAssetPtr=/NitroDrome/Assets/Audio/PA/Crowd/NitroDrome_PA_Crowd_Cheer_07}}
 {id=WavePlayer_12, properties {soundWaveAssetPtr=/NitroDrome/Assets/Audio/PA/Crowd/NitroDrome_PA_Crowd_Cheer_08}}
 {id=WavePlayer_14, properties {soundWaveAssetPtr=/NitroDrome/Assets/Audio/PA/Crowd/NitroDrome_PA_Crowd_Cheer_01}}
 {id=WavePlayer_15, properties {soundWaveAssetPtr=/NitroDrome/Assets/Audio/PA/Crowd/NitroDrome_PA_Crowd_Cheer_02}}
 {id=WavePlayer_16, properties {soundWaveAssetPtr=/NitroDrome/Assets/Audio/PA/Crowd/NitroDrome_PA_Crowd_Cheer_03}}
 {id=WavePlayer_17, properties {soundWaveAssetPtr=/NitroDrome/Assets/Audio/PA/Crowd/NitroDrome_PA_Crowd_Cheer_04}}
 {id=WavePlayer_18, properties {soundWaveAssetPtr=/NitroDrome/Assets/Audio/PA/Crowd/NitroDrome_PA_Crowd_Cheer_05}}
 {id=WavePlayer_22, properties {soundWaveAssetPtr=/NitroDrome/Assets/Audio/PA/Crowd/NitroDrome_PA_Crowd_Shock_01}}
 {id=WavePlayer_23, properties {soundWaveAssetPtr=/NitroDrome/Assets/Audio/PA/Crowd/NitroDrome_PA_Crowd_Shock_02}}
 {id=WavePlayer_24, properties {soundWaveAssetPtr=/NitroDrome/Assets/Audio/PA/Crowd/NitroDrome_PA_Crowd_Shock_03}}
 {id=WavePlayer_25, properties {soundWaveAssetPtr=/NitroDrome/Assets/Audio/PA/Crowd/NitroDrome_PA_Crowd_Shock_04}}
 {id=WavePlayer_26, properties {soundWaveAssetPtr=/NitroDrome/Assets/Audio/PA/Crowd/NitroDrome_PA_Crowd_Shock_05}}
 {id=WavePlayer_27, properties {soundWaveAssetPtr=/NitroDrome/Assets/Audio/PA/Crowd/NitroDrome_PA_Crowd_Shock_06}}
 {id=WavePlayer_4, properties {soundWaveAssetPtr=&/VO_Nitro_Queen_PA_Fire_Hazard_01}}
 {id=WavePlayer_5, properties {soundWaveAssetPtr=&/VO_Nitro_Queen_PA_Fire_Hazard_02}}
 {id=WavePlayer_6, properties {soundWaveAssetPtr=&/VO_Nitro_Queen_PA_Fire_Hazard_03}}
 {id=WavePlayer_7, properties {soundWaveAssetPtr=&/VO_Nitro_Queen_PA_Fire_Hazard_04}}
 {id=WavePlayer_8, properties {soundWaveAssetPtr=&/VO_Nitro_Queen_PA_Fire_Hazard_05}}
]
properties {
 attenuationSettings=/NitroDrome/Assets/Audio/PA/VO/VO_PA_Announcer_Att
 maxDistance=20000
 priority=1.05
 soundClassObject=&Fort_VO_NPC
 volumeMultiplier=0.85
 concurrencySet [/NitroDrome/Assets/Audio/PA/VO/VO_PA_Announcer_Conc]
}
