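# Sound cue for the NitroDrome PA that plays Nitro Queen player-elimination announcer lines with crowd reactions.
# A branch on the BossSpawn parameter plays only the crowd when true, and mixes a random voice line over the crowd when false or unset.
# The crowd layer randomly picks a cheer or shock pool, each behind a delay node, then a random clip from it.
# Announcer-specific attenuation and concurrency assets govern playback.
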SoundCue /NitroDrome/Assets/Audio/PA/VO/Nitro_Queen/VO_Nitro_Queen_PA_Player_Eliminated_Cue
root=Branch_0
nodes [
 {id=Branch_0, children [Mixer_2,Mixer_1,Mixer_1], properties {boolParameterName=BossSpawn}}
 {id=Delay_1, children [Random_6]}
 {id=Delay_3, children [Random_5]}
 {id=Mixer_1, children [Random_0,Random_7], properties {inputVolume [0.85,0.6]}}
 {id=Mixer_2, children [Random_7], properties {inputVolume [0.6]}}
 {
  id=Random_0
  children [WavePlayer_0, WavePlayer_1, WavePlayer_2, WavePlayer_3, WavePlayer_4, WavePlayer_9, WavePlayer_10]
  properties {bShouldExcludeFromBranchCulling=true, weights [1,1,1,1,1,1,1]}
 }
 {
  id=Random_5
  children [WavePlayer_32, WavePlayer_33, WavePlayer_34, WavePlayer_35, WavePlayer_36, WavePlayer_37, WavePlayer_38, WavePlayer_39]
  properties {bShouldExcludeFromBranchCulling=true, weights [1,1,1,1,1,1,1,1]}
 }
 {
  id=Random_6
  children [WavePlayer_26, WavePlayer_27, WavePlayer_28, WavePlayer_29, WavePlayer_30, WavePlayer_31]
  properties {bShouldExcludeFromBranchCulling=true, weights [1,1,1,1,1,1]}
 }
 {id=Random_7, children [Delay_3,Delay_1], properties {weights [1,1]}}
 {id=WavePlayer_0, properties {soundWaveAssetPtr=&/VO_Nitro_Queen_PA_Player_Elim_01}}
 {id=WavePlayer_1, properties {soundWaveAssetPtr=&/VO_Nitro_Queen_PA_Player_Elim_02}}
 {id=WavePlayer_10, properties {soundWaveAssetPtr=&/VO_Nitro_Queen_PA_Player_Elim_07}}
 {id=WavePlayer_2, properties {soundWaveAssetPtr=&/VO_Nitro_Queen_PA_Player_Elim_03}}
 {id=WavePlayer_26, properties {soundWaveAssetPtr=/NitroDrome/Assets/Audio/PA/Crowd/NitroDrome_PA_Crowd_Shock_01}}
 {id=WavePlayer_27, properties {soundWaveAssetPtr=/NitroDrome/Assets/Audio/PA/Crowd/NitroDrome_PA_Crowd_Shock_02}}
 {id=WavePlayer_28, properties {soundWaveAssetPtr=/NitroDrome/Assets/Audio/PA/Crowd/NitroDrome_PA_Crowd_Shock_03}}
 {id=WavePlayer_29, properties {soundWaveAssetPtr=/NitroDrome/Assets/Audio/PA/Crowd/NitroDrome_PA_Crowd_Shock_04}}
 {id=WavePlayer_3, properties {soundWaveAssetPtr=&/VO_Nitro_Queen_PA_Player_Elim_04}}
 {id=WavePlayer_30, properties {soundWaveAssetPtr=/NitroDrome/Assets/Audio/PA/Crowd/NitroDrome_PA_Crowd_Shock_05}}
 {id=WavePlayer_31, properties {soundWaveAssetPtr=/NitroDrome/Assets/Audio/PA/Crowd/NitroDrome_PA_Crowd_Shock_06}}
 {id=WavePlayer_32, properties {soundWaveAssetPtr=/NitroDrome/Assets/Audio/PA/Crowd/NitroDrome_PA_Crowd_Cheer_01}}
 {id=WavePlayer_33, properties {soundWaveAssetPtr=/NitroDrome/Assets/Audio/PA/Crowd/NitroDrome_PA_Crowd_Cheer_02}}
 {id=WavePlayer_34, properties {soundWaveAssetPtr=/NitroDrome/Assets/Audio/PA/Crowd/NitroDrome_PA_Crowd_Cheer_03}}
 {id=WavePlayer_35, properties {soundWaveAssetPtr=/NitroDrome/Assets/Audio/PA/Crowd/NitroDrome_PA_Crowd_Cheer_04}}
 {id=WavePlayer_36, properties {soundWaveAssetPtr=/NitroDrome/Assets/Audio/PA/Crowd/NitroDrome_PA_Crowd_Cheer_05}}
 {id=WavePlayer_37, properties {soundWaveAssetPtr=/NitroDrome/Assets/Audio/PA/Crowd/NitroDrome_PA_Crowd_Cheer_06}}
 {id=WavePlayer_38, properties {soundWaveAssetPtr=/NitroDrome/Assets/Audio/PA/Crowd/NitroDrome_PA_Crowd_Cheer_07}}
 {id=WavePlayer_39, properties {soundWaveAssetPtr=/NitroDrome/Assets/Audio/PA/Crowd/NitroDrome_PA_Crowd_Cheer_08}}
 {id=WavePlayer_4, properties {soundWaveAssetPtr=&/VO_Nitro_Queen_PA_Player_Elim_05}}
 {id=WavePlayer_9, properties {soundWaveAssetPtr=&/VO_Nitro_Queen_PA_Player_Elim_06}}
]
properties {
 attenuationSettings=/NitroDrome/Assets/Audio/PA/VO/VO_PA_Announcer_Att
 maxDistance=20000
 priority=1.05
 soundClassObject=&Fort_VO_NPC
 volumeMultiplier=0.85
 concurrencySet [/NitroDrome/Assets/Audio/PA/VO/VO_PA_Announcer_Conc]
}
