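# Sound cue for Nitro Queen PA announcer lines, named for a player entering a POI vehicle at NitroDrome, layered with crowd cheers.
# A branch on the BossSpawn bool plays only a quieter crowd cheer when true; false or unset mixes a random announcer line with the cheer.
# Cheers are picked at random after a short fixed delay.
# Playback uses the shared PA announcer attenuation and concurrency.
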SoundCue /NitroDrome/Assets/Audio/PA/VO/Nitro_Queen/VO_Nitro_Queen_PA_Player_Enters_POI_Veh_Cue
root=Branch_0
nodes [
 {id=Branch_0, children [Mixer_2,Mixer_1,Mixer_1], properties {boolParameterName=BossSpawn}}
 {id=Delay_2, children [Random_2], properties {delayMax=0.5, delayMin=0.5}}
 {id=Mixer_1, children [Random_0,Delay_2], properties {inputVolume [0.85,0.6]}}
 {id=Mixer_2, children [Delay_2], properties {inputVolume [0.6]}}
 {
  id=Random_0
  children [WavePlayer_0,WavePlayer_1,WavePlayer_2,WavePlayer_3,WavePlayer_7]
  properties {bShouldExcludeFromBranchCulling=true, weights [1,1,1,1,1]}
 }
 {
  id=Random_2
  children [WavePlayer_9, WavePlayer_13, WavePlayer_14, WavePlayer_15, WavePlayer_16, WavePlayer_17, WavePlayer_18, WavePlayer_19]
  properties {bShouldExcludeFromBranchCulling=true, weights [1,1,1,1,1,1,1,1]}
 }
 {id=WavePlayer_0, properties {soundWaveAssetPtr=&/VO_Nitro_Queen_PA_Player_Enters_Veh_01}}
 {id=WavePlayer_1, properties {soundWaveAssetPtr=&/VO_Nitro_Queen_PA_Player_Enters_Veh_02}}
 {id=WavePlayer_13, properties {soundWaveAssetPtr=/NitroDrome/Assets/Audio/PA/Crowd/NitroDrome_PA_Crowd_Cheer_02}}
 {id=WavePlayer_14, properties {soundWaveAssetPtr=/NitroDrome/Assets/Audio/PA/Crowd/NitroDrome_PA_Crowd_Cheer_03}}
 {id=WavePlayer_15, properties {soundWaveAssetPtr=/NitroDrome/Assets/Audio/PA/Crowd/NitroDrome_PA_Crowd_Cheer_04}}
 {id=WavePlayer_16, properties {soundWaveAssetPtr=/NitroDrome/Assets/Audio/PA/Crowd/NitroDrome_PA_Crowd_Cheer_05}}
 {id=WavePlayer_17, properties {soundWaveAssetPtr=/NitroDrome/Assets/Audio/PA/Crowd/NitroDrome_PA_Crowd_Cheer_06}}
 {id=WavePlayer_18, properties {soundWaveAssetPtr=/NitroDrome/Assets/Audio/PA/Crowd/NitroDrome_PA_Crowd_Cheer_07}}
 {id=WavePlayer_19, properties {soundWaveAssetPtr=/NitroDrome/Assets/Audio/PA/Crowd/NitroDrome_PA_Crowd_Cheer_08}}
 {id=WavePlayer_2, properties {soundWaveAssetPtr=&/VO_Nitro_Queen_PA_Player_Enters_Veh_03}}
 {id=WavePlayer_3, properties {soundWaveAssetPtr=&/VO_Nitro_Queen_PA_Player_Enters_Veh_04}}
 {id=WavePlayer_7, properties {soundWaveAssetPtr=&/VO_Nitro_Queen_PA_Player_Enters_Veh_05}}
 {id=WavePlayer_9, properties {soundWaveAssetPtr=/NitroDrome/Assets/Audio/PA/Crowd/NitroDrome_PA_Crowd_Cheer_01}}
]
properties {
 attenuationSettings=/NitroDrome/Assets/Audio/PA/VO/VO_PA_Announcer_Att
 maxDistance=20000
 priority=1.05
 soundClassObject=&Fort_VO_NPC
 volumeMultiplier=0.85
 concurrencySet [/NitroDrome/Assets/Audio/PA/VO/VO_PA_Announcer_Conc]
}
