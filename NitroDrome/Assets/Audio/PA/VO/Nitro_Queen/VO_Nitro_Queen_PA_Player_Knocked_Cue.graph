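# Sound cue for the Nitro Queen PA announcer's player-knocked callout in NitroDrome, with announcer attenuation and concurrency.
# A BossSpawn bool branch plays only a crowd reaction when true; when false or unset it mixes a random Player_Elim voice line over a quieter crowd.
# The crowd layer picks evenly between cheer and shock groups, each a random clip behind a Delay node.
# Voice and clip randomizers are excluded from branch culling.
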SoundCue /NitroDrome/Assets/Audio/PA/VO/Nitro_Queen/VO_Nitro_Queen_PA_Player_Knocked_Cue
root=Branch_0
nodes [
 {id=Branch_0, children [Mixer_2,Mixer_1,Mixer_1], properties {boolParameterName=BossSpawn}}
 {id=Delay_0, children [Random_6]}
 {id=Delay_2, children [Random_7]}
 {id=Mixer_1, children [Random_0,Random_8], properties {inputVolume [0.85,0.6]}}
 {id=Mixer_2, children [Random_8], properties {inputVolume [0.6]}}
 {
  id=Random_0
  children [WavePlayer_0,WavePlayer_1,WavePlayer_2,WavePlayer_3,WavePlayer_4,WavePlayer_5,WavePlayer_6]
  properties {bShouldExcludeFromBranchCulling=true, weights [1,1,1,1,1,1,1]}
 }
 {
  id=Random_6
  children [WavePlayer_40, WavePlayer_41, WavePlayer_42, WavePlayer_43, WavePlayer_44, WavePlayer_45, WavePlayer_46, WavePlayer_47]
  properties {bShouldExcludeFromBranchCulling=true, weights [1,1,1,1,1,1,1,1]}
 }
 {
  id=Random_7
  children [WavePlayer_48, WavePlayer_49, WavePlayer_50, WavePlayer_51, WavePlayer_52, WavePlayer_53]
  properties {bShouldExcludeFromBranchCulling=true, weights [1,1,1,1,1,1]}
 }
 {id=Random_8, children [Delay_0,Delay_2], properties {weights [1,1]}}
 {id=WavePlayer_0, properties {soundWaveAssetPtr=&/VO_Nitro_Queen_PA_Player_Elim_01}}
 {id=WavePlayer_1, properties {soundWaveAssetPtr=&/VO_Nitro_Queen_PA_Player_Elim_02}}
 {id=WavePlayer_2, properties {soundWaveAssetPtr=&/VO_Nitro_Queen_PA_Player_Elim_03}}
 {id=WavePlayer_3, properties {soundWaveAssetPtr=&/VO_Nitro_Queen_PA_Player_Elim_04}}
 {id=WavePlayer_4, properties {soundWaveAssetPtr=&/VO_Nitro_Queen_PA_Player_Elim_05}}
 {id=WavePlayer_40, properties {soundWaveAssetPtr=/NitroDrome/Assets/Audio/PA/Crowd/NitroDrome_PA_Crowd_Cheer_01}}
 {id=WavePlayer_41, properties {soundWaveAssetPtr=/NitroDrome/Assets/Audio/PA/Crowd/NitroDrome_PA_Crowd_Cheer_02}}
 {id=WavePlayer_42, properties {soundWaveAssetPtr=/NitroDrome/Assets/Audio/PA/Crowd/NitroDrome_PA_Crowd_Cheer_03}}
 {id=WavePlayer_43, properties {soundWaveAssetPtr=/NitroDrome/Assets/Audio/PA/Crowd/NitroDrome_PA_Crowd_Cheer_04}}
 {id=WavePlayer_44, properties {soundWaveAssetPtr=/NitroDrome/Assets/Audio/PA/Crowd/NitroDrome_PA_Crowd_Cheer_05}}
 {id=WavePlayer_45, properties {soundWaveAssetPtr=/NitroDrome/Assets/Audio/PA/Crowd/NitroDrome_PA_Crowd_Cheer_06}}
 {id=WavePlayer_46, properties {soundWaveAssetPtr=/NitroDrome/Assets/Audio/PA/Crowd/NitroDrome_PA_Crowd_Cheer_07}}
 {id=WavePlayer_47, properties {soundWaveAssetPtr=/NitroDrome/Assets/Audio/PA/Crowd/NitroDrome_PA_Crowd_Cheer_08}}
 {id=WavePlayer_48, properties {soundWaveAssetPtr=/NitroDrome/Assets/Audio/PA/Crowd/NitroDrome_PA_Crowd_Shock_01}}
 {id=WavePlayer_49, properties {soundWaveAssetPtr=/NitroDrome/Assets/Audio/PA/Crowd/NitroDrome_PA_Crowd_Shock_02}}
 {id=WavePlayer_5, properties {soundWaveAssetPtr=&/VO_Nitro_Queen_PA_Player_Elim_06}}
 {id=WavePlayer_50, properties {soundWaveAssetPtr=/NitroDrome/Assets/Audio/PA/Crowd/NitroDrome_PA_Crowd_Shock_03}}
 {id=WavePlayer_51, properties {soundWaveAssetPtr=/NitroDrome/Assets/Audio/PA/Crowd/NitroDrome_PA_Crowd_Shock_04}}
 {id=WavePlayer_52, properties {soundWaveAssetPtr=/NitroDrome/Assets/Audio/PA/Crowd/NitroDrome_PA_Crowd_Shock_05}}
 {id=WavePlayer_53, properties {soundWaveAssetPtr=/NitroDrome/Assets/Audio/PA/Crowd/NitroDrome_PA_Crowd_Shock_06}}
 {id=WavePlayer_6, properties {soundWaveAssetPtr=&/VO_Nitro_Queen_PA_Player_Elim_07}}
]
properties {
 attenuationSettings=/NitroDrome/Assets/Audio/PA/VO/VO_PA_Announcer_Att
 maxDistance=20000
 priority=1.05
 soundClassObject=&Fort_VO_NPC
 volumeMultiplier=0.85
 concurrencySet [/NitroDrome/Assets/Audio/PA/VO/VO_PA_Announcer_Conc]
}
